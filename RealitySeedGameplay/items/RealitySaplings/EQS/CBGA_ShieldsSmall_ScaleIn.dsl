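; Reality sapling actor (small shields) whose static mesh grows in on spawn and can shrink away before dying.
; BeginPlay hides the mesh at start scale, reveals it after a random delay, then a looping timer FInterps uniform scale, with Z stretched, until it snaps to target.
; Scale Out and Die reuses that timer toward start scale after random staggering.
; On arrival, authority calls SilentDie after a delay while remotes hide the mesh.
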(event EventBeginPlay
  (|CallParentFunction)
  (Rendering|SetVisibility (Variables|Default|GetStaticMesh) false false)
  (bind _output_get (Variables|Default|SetCur%20Min%20World%20Scale (Variables|Default|GetStart%20World%20Scale)))
  (Default|SetWeedScale _output_get)
  (Variables|Default|SetCur%20Max%20World%20Scale (Variables|Default|GetDest%20World%20Scale))
  (Variables|Default|SetHasReachedDestScale false)
  (Utilities|FlowControl|Delay (Math|Random|RandomFloatinRange 0.0 0.45))
  (Rendering|SetVisibility (Variables|Default|GetStaticMesh) true false)
  (bind _returnvalue (Utilities|Time|SetTimerbyEvent (delegate Custom|UpdateLerpToDest) (Variables|Default|GetLerpTimerInterval) true))
  (Variables|Default|SetLerpTimerHandle _returnvalue))

(event Custom|UpdateLerpToDest
  (Utilities|FlowControl|Sequence
    (:then_0
      ;[1 1 1] Lerp Scale as needed
      (if (not (Variables|Default|GetHasReachedDestScale))
        (Default|SetWeedScale (Math|Interpolation|FInterpTo (Variables|Default|GetCurr%20World%20Scale) (Variables|Default|GetCur%20Max%20World%20Scale) (Variables|Default|GetLerpTimerInterval) (Variables|Default|GetScale%20Interp%20Speed)))
        (if (<= (Math|Float|Absolute(Float) (- (Variables|Default|GetCurr%20World%20Scale) (Variables|Default|GetCur%20Max%20World%20Scale))) 0.02)
          (Default|SetWeedScale (Variables|Default|GetCur%20Max%20World%20Scale))
          (Variables|Default|SetHasReachedDestScale true))))
    (:then_1
      ;[1 1 1] Cancel looping once we've hit our target
      (if (Variables|Default|GetHasReachedDestScale)
        (Utilities|Time|ClearandInvalidateTimerbyHandle (Variables|Default|GetLerpTimerHandle))
        (if (Variables|Default|GetScaleOutAndDie)
          (Networking|SwitchHasAuthority
            (:Authority
              (Utilities|FlowControl|Delay 1.0)
              (Gameplay|SilentDie false))
            (:Remote
              (Rendering|SetVisibility (Variables|Default|GetStaticMesh) false true))))))))

(event Custom|Scale%20Out%20and%20Die
  (Variables|Default|SetScaleOutAndDie true)
  (Utilities|FlowControl|Delay (Math|Random|RandomFloatinRange 0.0 0.4))
  (Variables|Default|SetCur%20Min%20World%20Scale (Variables|Default|GetCurr%20World%20Scale))
  (Variables|Default|SetCur%20Max%20World%20Scale (Variables|Default|GetStart%20World%20Scale))
  (Variables|Default|SetHasReachedDestScale false)
  (bind _returnvalue (Utilities|Time|SetTimerbyEvent (delegate Custom|UpdateLerpToDest) (Variables|Default|GetLerpTimerInterval) true))
  (Variables|Default|SetLerpTimerHandle _returnvalue))

(event Collision|EventActorBeginOverlap
  (params (OtherActor :type Actor%20Object%20Reference)))

(event EventTick
  (params (DeltaSeconds :type Float%20%28single-precision%29)))

(fn SetWeedScale
  (params (SizeAlpha :type Float%20%28double-precision%29))
  (bind _output_get (Variables|Default|SetCurr%20World%20Scale SizeAlpha))
  (Transformation|SetWorldScale3D (Variables|Default|GetStaticMesh) (Math|Vector|MakeVector _output_get _output_get (* _output_get 1.25))))

(fn ConstructionScript
  (|CallParentFunction)
  (Variables|Default|SetDest%20World%20Scale 1.0)
  (Transformation|SetWorldScale3D (Variables|Default|GetStaticMesh) (Math|Vector|MakeVector (Variables|Default|GetDest%20World%20Scale) (Variables|Default|GetDest%20World%20Scale) (* (Variables|Default|GetDest%20World%20Scale) 1.25))))
