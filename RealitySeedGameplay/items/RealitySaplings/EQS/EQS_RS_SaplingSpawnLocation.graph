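# EQS query that picks a ground location for spawning a Reality Sapling near a supplied world location, favoring points closest to it.
# It tries a grid around that location first, then three spiral donut fallbacks that differ only in increasingly tall vertical projection traces.
# Every option repeats the same filters: water, destructibles, valid flat surface, spacing from other saplings, obstructions, avoided-actor bounds and the playable-bounds volume.
# Tuning values are bound to query parameters.
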EnvQuery /RealitySeedGameplay/items/RealitySaplings/EQS/EQS_RS_SaplingSpawnLocation
options [
 {
  generator {
   name=FortEnvQueryGenerator_SimpleGrid_1
   class=FortEnvQueryGenerator_SimpleGrid
   properties {
    bAutoSortTests=true
    bCanRunAsync=false
    envQueryResultNormalizationOption=Default
    generateAround=/Script/FortniteAI.FortQueryContext_WorldLocationParam
    navDataOverrideContext=null
    optionName=""
    overridenAgentNameForNavmesh=None
    randomDisplacementRatio=0
    gridSize {
     dataBinding=&self:FortEnvQueryGenerator_SimpleGrid_1.AIDataProvider_QueryParams_0
     dataField=FloatValue
     defaultValue=500
    }
    projectionData {
     bCanDisableTrace=true
     bCanProjectDown=true
     bCanTraceOnGeometry=true
     bCanTraceOnNavMesh=true
     bOnlyBlockingHits=true
     bTraceComplex=false
     extentX=0
     extentY=10
     extentZ=10
     navigationFilter=null
     postProjectionVerticalOffset=0
     projectDown=2000
     projectUp=2000
     serializedChannel=ECC_GameTraceChannel6
     traceChannel=TraceTypeQuery7
     traceMode=GeometryByChannel
     traceProfileName=None
     traceShape=Line
     versionNum=1
    }
    spaceBetween {
     dataBinding=&self:FortEnvQueryGenerator_SimpleGrid_1.AIDataProvider_QueryParams_1
     dataField=FloatValue
     defaultValue=100
    }
   }
  }
  tests [
   EnvQueryTest_Overlap_6 {
    class=EnvQueryTest_Overlap
    properties {testComment=Reject Water, overlapData {extentX=250, extentY=250, extentZ=250, overlapChannel=ECC_GameTraceChannel15, shapeOffset=<0 0 200>}}
   }
   EnvQueryTest_Overlap_7 {
    class=EnvQueryTest_Overlap
    properties {
     testComment=Reject Destructibels
     overlapData {extentX=250, extentY=250, extentZ=250, overlapChannel=ECC_GameTraceChannel11, shapeOffset=<0 0 265>}
    }
   }
   FortQueryTest_ValidSurface_0 {
    class=FortQueryTest_ValidSurface
    properties {
     flatSurfaceToleranceZ {
      dataBinding=&self:FortQueryTest_ValidSurface_0.AIDataProvider_QueryParams_3
      dataField=FloatValue
      defaultValue=10
     }
     invalidHitActorClasses [
      /Script/FortniteGame.BuildingDeco
      &BuildingProp
      /Script/FortniteGame.BuildingGameplayActor
      &BuildingStairs
      &BuildingWall
     ]
     radius {
      dataBinding=&self:FortQueryTest_ValidSurface_0.AIDataProvider_QueryParams_0
      dataField=FloatValue
      defaultValue=250
     }
     surfaceMaterials [
      /Game/FortressPhysicalMaterials/Asphalt
      /Engine/EngineMaterials/DefaultPhysicalMaterial
      /Game/FortressPhysicalMaterials/Generic
      /Game/FortressPhysicalMaterials/Metal
     ]
     traceOffsetDown {
      dataBinding=&self:FortQueryTest_ValidSurface_0.AIDataProvider_QueryParams_2
      dataField=FloatValue
      defaultValue=100
     }
     traceOffsetUp {
      dataBinding=&self:FortQueryTest_ValidSurface_0.AIDataProvider_QueryParams_1
      dataField=FloatValue
      defaultValue=100
     }
     validHitActorClasses [
      /Script/Landscape.LandscapeStreamingProxy
      /Script/Landscape
      /Script/FortniteGame.BuildingActor
      &StaticMeshActor
      /Script/Engine.SplineMeshActor
      /Script/Landscape.ControlPointMeshActor
     ]
    }
   }
   EnvQueryTest_Distance_0 {
    class=EnvQueryTest_Distance
    properties {
     distanceTo=/Script/RealitySeedGameplayRuntime.FortQueryContext_RealitySaplingLocations
     filterType=Minimum
     scoringEquation=Linear
     scoringFactor=1
     testComment=""
     testMode=Distance2D
     testPurpose=Filter
     floatValueMin {
      dataBinding=&self:EnvQueryTest_Distance_0.AIDataProvider_QueryParams_0
      dataField=FloatValue
      defaultValue=0
     }
    }
   }
   EnvQueryTest_Distance_6 {
    class=EnvQueryTest_Distance
    properties {
     distanceTo=&FortQueryContext_WorldLocationParam
     filterType=Range
     floatValueMin=0
     scoringEquation=Linear
     scoringFactor=-1
     testComment=""
     testMode=Distance2D
     testPurpose=Score
    }
   }
   FortQueryTest_RealitySaplingObstructed_0 {
    class=FortQueryTest_RealitySaplingObstructed
    properties {
     bDefineReferenceValue=false
     bOnlyBlockingHits=true
     bOverlapComplex=false
     boolValue=true
     clampMaxType=None
     clampMinType=None
     filterType=Match
     floatValueMax=0
     floatValueMin=0
     multipleContextFilterOp=AllPass
     multipleContextScoreOp=AverageScore
     normalizationType=Absolute
     referenceValue=0
     scoreClampMax=0
     scoreClampMin=0
     scoringEquation=Constant
     scoringFactor=1
     testComment=""
     testPurpose=Filter
     ignoreActorClasses [&Landscape, &LandscapeStreamingProxy]
     objectTypesToQuery [ECC_WorldStatic, ECC_Pawn, ECC_Vehicle, ECC_GameTraceChannel17]
     sphereOverlapRadius {
      dataBinding=&self:FortQueryTest_RealitySaplingObstructed_0.AIDataProvider_QueryParams_0
      dataField=FloatValue
      defaultValue=250
     }
     sphereOverlapZOffset {
      dataBinding=&self:FortQueryTest_RealitySaplingObstructed_0.AIDataProvider_QueryParams_1
      dataField=FloatValue
      defaultValue=265
     }
    }
   }
   FortQueryTest_DistanceToActorBound_0 {
    class=FortQueryTest_DistanceToActorBound
    properties {
     bDefineReferenceValue=false
     bUse2DDistance=false
     boolValue=true
     clampMaxType=None
     clampMinType=None
     distanceTo=&/EnvQueryContext_SaplingSpawnActorsToAvoidByBounds.EnvQueryContext_SaplingSpawnActorsToAvoidByBounds_C
     filterType=Minimum
     floatValueMax=0
     multipleContextFilterOp=AllPass
     multipleContextScoreOp=AverageScore
     normalizationType=Absolute
     referenceValue=0
     scoreClampMax=0
     scoreClampMin=0
     scoringEquation=Linear
     scoringFactor=1
     testComment=""
     testPurpose=Filter
     floatValueMin {
      dataBinding=&self:FortQueryTest_DistanceToActorBound_0.AIDataProvider_QueryParams_0
      dataField=FloatValue
      defaultValue=300
     }
    }
   }
   EnvQueryTest_Volume_0 {
    class=EnvQueryTest_Volume
    properties {
     boolValue=true
     testPurpose=Filter
     volumeClass=null
     volumeContext=&/EnvQueryContext_SaplingSpawn_PlayableBoundsVolume.EnvQueryContext_SaplingSpawn_PlayableBoundsVolume_C
    }
   }
  ]
 }
 {
  generator {
   name=EnvQueryGenerator_Donut_2
   class=EnvQueryGenerator_Donut
   properties {
    bUseSpiralPattern=true
    center=&FortQueryContext_WorldLocationParam
    optionName=Donut Fallback 1
    innerRadius {
     dataBinding=&self:EnvQueryGenerator_Donut_2.AIDataProvider_QueryParams_0
     dataField=FloatValue
     defaultValue=5000
    }
    numberOfRings {
     dataBinding=&self:EnvQueryGenerator_Donut_2.AIDataProvider_QueryParams_2
     dataField=IntValue
     defaultValue=5
    }
    outerRadius {
     dataBinding=&self:EnvQueryGenerator_Donut_2.AIDataProvider_QueryParams_1
     dataField=FloatValue
     defaultValue=10000
    }
    pointsPerRing {
     dataBinding=&self:EnvQueryGenerator_Donut_2.AIDataProvider_QueryParams_3
     dataField=IntValue
     defaultValue=30
    }
    projectionData {
     extentX=0
     navigationFilter=null
     postProjectionVerticalOffset=0
     projectDown=10000
     projectUp=10000
     serializedChannel=ECC_GameTraceChannel6
     traceChannel=TraceTypeQuery7
     traceMode=GeometryByChannel
    }
   }
  }
  tests [
   EnvQueryTest_Overlap_0 {
    class=EnvQueryTest_Overlap
    properties {testComment=Reject Water, overlapData {extentX=250, extentY=250, extentZ=250, overlapChannel=ECC_GameTraceChannel15, shapeOffset=<0 0 200>}}
   }
   EnvQueryTest_Overlap_1 {
    class=EnvQueryTest_Overlap
    properties {
     testComment=Reject Destructibels
     overlapData {extentX=250, extentY=250, extentZ=250, overlapChannel=ECC_GameTraceChannel11, shapeOffset=<0 0 265>}
    }
   }
   FortQueryTest_ValidSurface_1 {
    class=FortQueryTest_ValidSurface
    properties {
     flatSurfaceToleranceZ {
      dataBinding=&self:FortQueryTest_ValidSurface_1.AIDataProvider_QueryParams_3
      dataField=FloatValue
      defaultValue=10
     }
     invalidHitActorClasses [
      &BuildingDeco
      &BuildingProp
      &BuildingGameplayActor
      &BuildingStairs
      &BuildingWall
     ]
     radius {
      dataBinding=&self:FortQueryTest_ValidSurface_1.AIDataProvider_QueryParams_0
      dataField=FloatValue
      defaultValue=250
     }
     surfaceMaterials [
      &Asphalt
      &DefaultPhysicalMaterial
      &Generic
      &Metal
     ]
     traceOffsetDown {
      dataBinding=&self:FortQueryTest_ValidSurface_1.AIDataProvider_QueryParams_2
      dataField=FloatValue
      defaultValue=100
     }
     traceOffsetUp {
      dataBinding=&self:FortQueryTest_ValidSurface_1.AIDataProvider_QueryParams_1
      dataField=FloatValue
      defaultValue=100
     }
     validHitActorClasses [
      &LandscapeStreamingProxy
      &Landscape
      &BuildingActor
      &StaticMeshActor
      &SplineMeshActor
      &ControlPointMeshActor
     ]
    }
   }
   EnvQueryTest_Distance_1 {
    class=EnvQueryTest_Distance
    properties {
     distanceTo=&FortQueryContext_RealitySaplingLocations
     filterType=Minimum
     scoringEquation=Linear
     scoringFactor=1
     testComment=""
     testMode=Distance2D
     testPurpose=Filter
     floatValueMin {
      dataBinding=&self:EnvQueryTest_Distance_1.AIDataProvider_QueryParams_0
      dataField=FloatValue
      defaultValue=0
     }
    }
   }
   EnvQueryTest_Distance_2 {
    class=EnvQueryTest_Distance
    properties {
     distanceTo=&FortQueryContext_WorldLocationParam
     filterType=Range
     floatValueMin=0
     scoringEquation=Linear
     scoringFactor=-1
     testComment=""
     testMode=Distance2D
     testPurpose=Score
    }
   }
   FortQueryTest_RealitySaplingObstructed_1 {
    class=FortQueryTest_RealitySaplingObstructed
    properties {
     bDefineReferenceValue=false
     bOnlyBlockingHits=true
     bOverlapComplex=false
     boolValue=true
     clampMaxType=None
     clampMinType=None
     filterType=Match
     floatValueMax=0
     floatValueMin=0
     multipleContextFilterOp=AllPass
     multipleContextScoreOp=AverageScore
     normalizationType=Absolute
     referenceValue=0
     scoreClampMax=0
     scoreClampMin=0
     scoringEquation=Constant
     scoringFactor=1
     testComment=""
     testPurpose=Filter
     ignoreActorClasses [&Landscape, &LandscapeStreamingProxy]
     objectTypesToQuery [ECC_WorldStatic, ECC_Pawn, ECC_Vehicle, ECC_GameTraceChannel17]
     sphereOverlapRadius {
      dataBinding=&self:FortQueryTest_RealitySaplingObstructed_1.AIDataProvider_QueryParams_0
      dataField=FloatValue
      defaultValue=250
     }
     sphereOverlapZOffset {
      dataBinding=&self:FortQueryTest_RealitySaplingObstructed_1.AIDataProvider_QueryParams_1
      dataField=FloatValue
      defaultValue=265
     }
    }
   }
   FortQueryTest_DistanceToActorBound_1 {
    class=FortQueryTest_DistanceToActorBound
    properties {
     bDefineReferenceValue=false
     bUse2DDistance=false
     boolValue=true
     clampMaxType=None
     clampMinType=None
     distanceTo=&EnvQueryContext_SaplingSpawnActorsToAvoidByBounds_C
     filterType=Minimum
     floatValueMax=0
     multipleContextFilterOp=AllPass
     multipleContextScoreOp=AverageScore
     normalizationType=Absolute
     referenceValue=0
     scoreClampMax=0
     scoreClampMin=0
     scoringEquation=Linear
     scoringFactor=1
     testComment=""
     testPurpose=Filter
     floatValueMin {
      dataBinding=&self:FortQueryTest_DistanceToActorBound_1.AIDataProvider_QueryParams_0
      dataField=FloatValue
      defaultValue=300
     }
    }
   }
   EnvQueryTest_Volume_1 {
    class=EnvQueryTest_Volume
    properties {
     boolValue=true
     testPurpose=Filter
     volumeClass=null
     volumeContext=&EnvQueryContext_SaplingSpawn_PlayableBoundsVolume_C
    }
   }
  ]
 }
 {
  generator {
   name=EnvQueryGenerator_Donut_0
   class=EnvQueryGenerator_Donut
   properties {
    bUseSpiralPattern=true
    center=&FortQueryContext_WorldLocationParam
    optionName=Donut Fallback 2
    innerRadius {
     dataBinding=&self:EnvQueryGenerator_Donut_0.AIDataProvider_QueryParams_0
     dataField=FloatValue
     defaultValue=5000
    }
    numberOfRings {
     dataBinding=&self:EnvQueryGenerator_Donut_0.AIDataProvider_QueryParams_2
     dataField=IntValue
     defaultValue=5
    }
    outerRadius {
     dataBinding=&self:EnvQueryGenerator_Donut_0.AIDataProvider_QueryParams_1
     dataField=FloatValue
     defaultValue=10000
    }
    pointsPerRing {
     dataBinding=&self:EnvQueryGenerator_Donut_0.AIDataProvider_QueryParams_3
     dataField=IntValue
     defaultValue=30
    }
    projectionData {
     extentX=0
     navigationFilter=null
     postProjectionVerticalOffset=0
     projectDown=20000
     projectUp=20000
     serializedChannel=ECC_GameTraceChannel6
     traceChannel=TraceTypeQuery7
     traceMode=GeometryByChannel
    }
   }
  }
  tests [
   EnvQueryTest_Overlap_2 {
    class=EnvQueryTest_Overlap
    properties {testComment=Reject Water, overlapData {extentX=250, extentY=250, extentZ=250, overlapChannel=ECC_GameTraceChannel15, shapeOffset=<0 0 200>}}
   }
   EnvQueryTest_Overlap_3 {
    class=EnvQueryTest_Overlap
    properties {
     testComment=Reject Destructibels
     overlapData {extentX=250, extentY=250, extentZ=250, overlapChannel=ECC_GameTraceChannel11, shapeOffset=<0 0 265>}
    }
   }
   FortQueryTest_ValidSurface_2 {
    class=FortQueryTest_ValidSurface
    properties {
     flatSurfaceToleranceZ {
      dataBinding=&self:FortQueryTest_ValidSurface_2.AIDataProvider_QueryParams_3
      dataField=FloatValue
      defaultValue=10
     }
     invalidHitActorClasses [
      &BuildingDeco
      &BuildingProp
      &BuildingGameplayActor
      &BuildingStairs
      &BuildingWall
     ]
     radius {
      dataBinding=&self:FortQueryTest_ValidSurface_2.AIDataProvider_QueryParams_0
      dataField=FloatValue
      defaultValue=250
     }
     surfaceMaterials [
      &Asphalt
      &DefaultPhysicalMaterial
      &Generic
      &Metal
     ]
     traceOffsetDown {
      dataBinding=&self:FortQueryTest_ValidSurface_2.AIDataProvider_QueryParams_2
      dataField=FloatValue
      defaultValue=100
     }
     traceOffsetUp {
      dataBinding=&self:FortQueryTest_ValidSurface_2.AIDataProvider_QueryParams_1
      dataField=FloatValue
      defaultValue=100
     }
     validHitActorClasses [
      &LandscapeStreamingProxy
      &Landscape
      &BuildingActor
      &StaticMeshActor
      &SplineMeshActor
      &ControlPointMeshActor
     ]
    }
   }
   EnvQueryTest_Distance_3 {
    class=EnvQueryTest_Distance
    properties {
     distanceTo=&FortQueryContext_RealitySaplingLocations
     filterType=Minimum
     scoringEquation=Linear
     scoringFactor=1
     testComment=""
     testMode=Distance2D
     testPurpose=Filter
     floatValueMin {
      dataBinding=&self:EnvQueryTest_Distance_3.AIDataProvider_QueryParams_0
      dataField=FloatValue
      defaultValue=0
     }
    }
   }
   EnvQueryTest_Distance_4 {
    class=EnvQueryTest_Distance
    properties {
     distanceTo=&FortQueryContext_WorldLocationParam
     filterType=Range
     floatValueMin=0
     scoringEquation=Linear
     scoringFactor=-1
     testComment=""
     testMode=Distance2D
     testPurpose=Score
    }
   }
   FortQueryTest_RealitySaplingObstructed_2 {
    class=FortQueryTest_RealitySaplingObstructed
    properties {
     bDefineReferenceValue=false
     bOnlyBlockingHits=true
     bOverlapComplex=false
     boolValue=true
     clampMaxType=None
     clampMinType=None
     filterType=Match
     floatValueMax=0
     floatValueMin=0
     multipleContextFilterOp=AllPass
     multipleContextScoreOp=AverageScore
     normalizationType=Absolute
     referenceValue=0
     scoreClampMax=0
     scoreClampMin=0
     scoringEquation=Constant
     scoringFactor=1
     testComment=""
     testPurpose=Filter
     ignoreActorClasses [&Landscape, &LandscapeStreamingProxy]
     objectTypesToQuery [ECC_WorldStatic, ECC_Pawn, ECC_Vehicle, ECC_GameTraceChannel17]
     sphereOverlapRadius {
      dataBinding=&self:FortQueryTest_RealitySaplingObstructed_2.AIDataProvider_QueryParams_0
      dataField=FloatValue
      defaultValue=250
     }
     sphereOverlapZOffset {
      dataBinding=&self:FortQueryTest_RealitySaplingObstructed_2.AIDataProvider_QueryParams_1
      dataField=FloatValue
      defaultValue=265
     }
    }
   }
   FortQueryTest_DistanceToActorBound_2 {
    class=FortQueryTest_DistanceToActorBound
    properties {
     bDefineReferenceValue=false
     bUse2DDistance=false
     boolValue=true
     clampMaxType=None
     clampMinType=None
     distanceTo=&EnvQueryContext_SaplingSpawnActorsToAvoidByBounds_C
     filterType=Minimum
     floatValueMax=0
     multipleContextFilterOp=AllPass
     multipleContextScoreOp=AverageScore
     normalizationType=Absolute
     referenceValue=0
     scoreClampMax=0
     scoreClampMin=0
     scoringEquation=Linear
     scoringFactor=1
     testComment=""
     testPurpose=Filter
     floatValueMin {
      dataBinding=&self:FortQueryTest_DistanceToActorBound_2.AIDataProvider_QueryParams_0
      dataField=FloatValue
      defaultValue=300
     }
    }
   }
   EnvQueryTest_Volume_2 {
    class=EnvQueryTest_Volume
    properties {
     boolValue=true
     testPurpose=Filter
     volumeClass=null
     volumeContext=&EnvQueryContext_SaplingSpawn_PlayableBoundsVolume_C
    }
   }
  ]
 }
 {
  generator {
   name=EnvQueryGenerator_Donut_1
   class=EnvQueryGenerator_Donut
   properties {
    bUseSpiralPattern=true
    center=&FortQueryContext_WorldLocationParam
    optionName=Donut Fallback 3
    innerRadius {
     dataBinding=&self:EnvQueryGenerator_Donut_1.AIDataProvider_QueryParams_0
     dataField=FloatValue
     defaultValue=5000
    }
    numberOfRings {
     dataBinding=&self:EnvQueryGenerator_Donut_1.AIDataProvider_QueryParams_2
     dataField=IntValue
     defaultValue=5
    }
    outerRadius {
     dataBinding=&self:EnvQueryGenerator_Donut_1.AIDataProvider_QueryParams_1
     dataField=FloatValue
     defaultValue=10000
    }
    pointsPerRing {
     dataBinding=&self:EnvQueryGenerator_Donut_1.AIDataProvider_QueryParams_3
     dataField=IntValue
     defaultValue=30
    }
    projectionData {
     extentX=0
     navigationFilter=null
     postProjectionVerticalOffset=0
     projectDown=50000
     projectUp=50000
     serializedChannel=ECC_GameTraceChannel6
     traceChannel=TraceTypeQuery7
     traceMode=GeometryByChannel
    }
   }
  }
  tests [
   EnvQueryTest_Overlap_4 {
    class=EnvQueryTest_Overlap
    properties {testComment=Reject Water, overlapData {extentX=250, extentY=250, extentZ=250, overlapChannel=ECC_GameTraceChannel15, shapeOffset=<0 0 200>}}
   }
   EnvQueryTest_Overlap_5 {
    class=EnvQueryTest_Overlap
    properties {
     testComment=Reject Destructibels
     overlapData {extentX=250, extentY=250, extentZ=250, overlapChannel=ECC_GameTraceChannel11, shapeOffset=<0 0 265>}
    }
   }
   FortQueryTest_ValidSurface_3 {
    class=FortQueryTest_ValidSurface
    properties {
     flatSurfaceToleranceZ {
      dataBinding=&self:FortQueryTest_ValidSurface_3.AIDataProvider_QueryParams_3
      dataField=FloatValue
      defaultValue=10
     }
     invalidHitActorClasses [
      &BuildingDeco
      &BuildingProp
      &BuildingGameplayActor
      &BuildingStairs
      &BuildingWall
     ]
     radius {
      dataBinding=&self:FortQueryTest_ValidSurface_3.AIDataProvider_QueryParams_0
      dataField=FloatValue
      defaultValue=250
     }
     surfaceMaterials [
      &Asphalt
      &DefaultPhysicalMaterial
      &Generic
      &Metal
     ]
     traceOffsetDown {
      dataBinding=&self:FortQueryTest_ValidSurface_3.AIDataProvider_QueryParams_2
      dataField=FloatValue
      defaultValue=100
     }
     traceOffsetUp {
      dataBinding=&self:FortQueryTest_ValidSurface_3.AIDataProvider_QueryParams_1
      dataField=FloatValue
      defaultValue=100
     }
     validHitActorClasses [
      &LandscapeStreamingProxy
      &Landscape
      &BuildingActor
      &StaticMeshActor
      &SplineMeshActor
      &ControlPointMeshActor
     ]
    }
   }
   EnvQueryTest_Distance_8 {
    class=EnvQueryTest_Distance
    properties {
     distanceTo=&FortQueryContext_RealitySaplingLocations
     filterType=Minimum
     scoringEquation=Linear
     scoringFactor=1
     testComment=""
     testMode=Distance2D
     testPurpose=Filter
     floatValueMin {
      dataBinding=&self:EnvQueryTest_Distance_8.AIDataProvider_QueryParams_0
      dataField=FloatValue
      defaultValue=0
     }
    }
   }
   EnvQueryTest_Distance_9 {
    class=EnvQueryTest_Distance
    properties {
     distanceTo=&FortQueryContext_WorldLocationParam
     filterType=Range
     floatValueMin=0
     scoringEquation=Linear
     scoringFactor=-1
     testComment=""
     testMode=Distance2D
     testPurpose=Score
    }
   }
   FortQueryTest_RealitySaplingObstructed_3 {
    class=FortQueryTest_RealitySaplingObstructed
    properties {
     bDefineReferenceValue=false
     bOnlyBlockingHits=true
     bOverlapComplex=false
     boolValue=true
     clampMaxType=None
     clampMinType=None
     filterType=Match
     floatValueMax=0
     floatValueMin=0
     multipleContextFilterOp=AllPass
     multipleContextScoreOp=AverageScore
     normalizationType=Absolute
     referenceValue=0
     scoreClampMax=0
     scoreClampMin=0
     scoringEquation=Constant
     scoringFactor=1
     testComment=""
     testPurpose=Filter
     ignoreActorClasses [&Landscape, &LandscapeStreamingProxy]
     objectTypesToQuery [ECC_WorldStatic, ECC_Pawn, ECC_Vehicle, ECC_GameTraceChannel17]
     sphereOverlapRadius {
      dataBinding=&self:FortQueryTest_RealitySaplingObstructed_3.AIDataProvider_QueryParams_0
      dataField=FloatValue
      defaultValue=250
     }
     sphereOverlapZOffset {
      dataBinding=&self:FortQueryTest_RealitySaplingObstructed_3.AIDataProvider_QueryParams_1
      dataField=FloatValue
      defaultValue=265
     }
    }
   }
   FortQueryTest_DistanceToActorBound_3 {
    class=FortQueryTest_DistanceToActorBound
    properties {
     bDefineReferenceValue=false
     bUse2DDistance=false
     boolValue=true
     clampMaxType=None
     clampMinType=None
     distanceTo=&EnvQueryContext_SaplingSpawnActorsToAvoidByBounds_C
     filterType=Minimum
     floatValueMax=0
     multipleContextFilterOp=AllPass
     multipleContextScoreOp=AverageScore
     normalizationType=Absolute
     referenceValue=0
     scoreClampMax=0
     scoreClampMin=0
     scoringEquation=Linear
     scoringFactor=1
     testComment=""
     testPurpose=Filter
     floatValueMin {
      dataBinding=&self:FortQueryTest_DistanceToActorBound_3.AIDataProvider_QueryParams_0
      dataField=FloatValue
      defaultValue=300
     }
    }
   }
   EnvQueryTest_Volume_3 {
    class=EnvQueryTest_Volume
    properties {
     boolValue=true
     testPurpose=Filter
     volumeClass=null
     volumeContext=&EnvQueryContext_SaplingSpawn_PlayableBoundsVolume_C
    }
   }
  ]
 }
]
