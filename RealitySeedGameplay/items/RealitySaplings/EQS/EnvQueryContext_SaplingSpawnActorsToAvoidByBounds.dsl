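; EQS context blueprint that supplies the set of actors which Reality Sapling spawn queries avoid by their bounds.
; ProvideActorsSet gets the Athena game state and its FortGameStateComponent_RealitySapling component.
; When both are valid it returns that component's avoid-by-bounds actor list, so the list itself is maintained outside this asset.
; An invalid game state or component falls through to an empty default return.
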(fn ProvideActorsSet
  (params (QuerierObject :type Object%20Reference) (QuerierActor :type Actor%20Object%20Reference))
  (bind _returnvalue (Fortnite|Athena|GameState|GetGameStateAthena))
  (bind _returnvalue_2 (Actor|GetComponentbyClass _returnvalue "/Script/RealitySeedGameplayRuntime.FortGameStateComponent_RealitySapling"))
  (Utilities|IsValid _returnvalue
    (:Is%20Valid
      (Utilities|IsValid _returnvalue_2
        (:Is%20Valid
          (return (Class|FortGameStateComponentRealitySapling|GetSaplingSpawnEnvQueryActorstoAvoidbyBounds _returnvalue_2)))
        (:Is%20Not%20Valid
          _)))
    (:Is%20Not%20Valid
      _))
  (return 0))
