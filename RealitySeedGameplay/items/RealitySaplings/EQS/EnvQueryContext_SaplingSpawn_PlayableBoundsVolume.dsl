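; EQS context for Reality Sapling spawn queries that supplies the map's playable bounds volume as its actor set.
; ProvideActorsSet reads the Athena game state, its MapInfo, then that MapInfo's PlayableBoundsVolume, ignoring the querier.
; Each link is validity-checked, and a valid volume is returned as a single-element array.
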(fn ProvideActorsSet
  (params (QuerierObject :type Object%20Reference) (QuerierActor :type Actor%20Object%20Reference))
  (bind _returnvalue (Fortnite|Athena|GameState|GetGameStateAthena))
  (bind _mapinfo (Variables|/Script/FortniteGame.FortGameStateAthena|GetMapInfo _returnvalue))
  (bind _playableboundsvolume (Variables|/Script/FortniteGame.FortAthenaMapInfo|GetPlayableBoundsVolume _mapinfo))
  (Utilities|IsValid _returnvalue
    (:Is%20Valid
      (Utilities|IsValid _mapinfo
        (:Is%20Valid
          (Utilities|IsValid _playableboundsvolume
            (:Is%20Valid
              (return (Utilities|Array|MakeArray _playableboundsvolume)))
            (:Is%20Not%20Valid
              _)))
        (:Is%20Not%20Valid
          _))
      (return 0))
    (:Is%20Not%20Valid
      _)))
